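# Material function for the EarthLane backpack FX that produces a two-channel UV distortion offset, named for flame effects.
# It samples the Distortion Texture (Voronoi height by default) twice with separately panned and scaled UVs, each using only the first component of its speed and scale inputs for both axes.
# A particle dynamic parameter offsets one sample in U, an input the other.
# Their red and green channels are multiplied, then biased down toward zero.
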MaterialFunction /Game/Accessories/FORT_Backpacks/Backpack_M_MED_EarthLane/FX/MF_FlameDistortion
graph {nodes [
  n1 = DynamicParameter() {defaultValue=<rgba 0 1 0 1>, paramNames [UV1 U Offset, Dive, Camera Offset, Param4]}
  n2 = Constant()
  n3 = FunctionInput() {bUsePreviewValueAsDefault=true, inputName=UV2 Speed, inputType=FunctionInput_Vector2, sortPriority=2, previewValue=<0 1 0 1>}
  n4 = MaterialFunctionCall(Float2=@n3) {materialFunction=&BreakOutFloat2Components}
  n5 = Panner(Coordinate=TextureCoordinate(), Speed=AppendVector(A=@n4.R, B=@n4.R)) {speedX=-0.5}
  n6 = FunctionInput() {bUsePreviewValueAsDefault=true, inputName=UV2 Scale, inputType=FunctionInput_Vector2, sortPriority=3, previewValue=<1 0.25 0 1>}
  n7 = MaterialFunctionCall(Float2=@n6) {materialFunction=&BreakOutFloat2Components}
  n8 = AppendVector(A=@n7.R, B=@n7.R)
  n9 = TextureSampleParameter2D(UVs=Add(A=AppendVector(A=@n1.UV1 U Offset, B=@n2), B=Multiply(A=@n5, B=@n8))) {group=UV Distortion, parameterName=Distortion Texture, texture=/Game/Effects/Fort_Effects/Textures/T_PinkSpike_Voronoi2D_Height}
  n10 = FunctionInput() {inputName=U Offset, inputType=FunctionInput_Scalar, sortPriority=5}
  n11 = Constant()
  n12 = FunctionInput() {bUsePreviewValueAsDefault=true, inputName=UV Speed, inputType=FunctionInput_Vector2, previewValue=<0 2 0 1>}
  n13 = MaterialFunctionCall(Float2=@n12) {materialFunction=&BreakOutFloat2Components}
  n14 = Panner(Coordinate=TextureCoordinate(), Speed=AppendVector(A=@n13.R, B=@n13.R)) {speedX=-0.5}
  n15 = FunctionInput() {bUsePreviewValueAsDefault=true, inputName=UV Scale, inputType=FunctionInput_Vector2, sortPriority=1, previewValue=<1 0.5 0 1>}
  n16 = MaterialFunctionCall(Float2=@n15) {materialFunction=&BreakOutFloat2Components}
  n17 = AppendVector(A=@n16.R, B=@n16.R)
  n18 = TextureSampleParameter2D(UVs=Add(A=AppendVector(A=@n10, B=@n11), B=Multiply(A=@n14, B=@n17))) {group=UV Distortion, parameterName=Distortion Texture, texture=&T_PinkSpike_Voronoi2D_Height}
  n19 = Multiply(A=@n9.R, B=@n18.R)
  n20 = Multiply(A=@n9.G, B=@n18.G)
  FunctionOutput(None=Subtract(A=MaterialFunctionCall(X=@n19, Y=@n20) {materialFunction=&MakeFloat2}.Result) {constB=0.25})
 ]}
