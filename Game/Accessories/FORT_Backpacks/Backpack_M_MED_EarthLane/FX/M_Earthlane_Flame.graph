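# Translucent, unlit, two-sided material for the flame VFX on the EarthLane backpack, usable on Niagara sprites, ribbons and meshes.
# Two-layer panning distortion, offset by a dynamic parameter, warps the MF_Flame mask and vertical gradient.
# Both sample a toon curve atlas for emissive color, scaled by time of day and optional particle color.
# Opacity adds view-angle fade and non-mobile depth fade; a static switch enables interpolated WPO.
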Material /Game/Accessories/FORT_Backpacks/Backpack_M_MED_EarthLane/FX/M_Earthlane_Flame
graph {
 properties {
  bUsedWithNiagaraMeshParticles=true
  bUsedWithNiagaraRibbons=true
  bUsedWithNiagaraSprites=true
  blendMode=BLEND_Translucent
  pixelDepthOffsetMode=PDOM_Legacy
  shadingModel=MSM_Unlit
  twoSided=true
  usedShadingModels=Unlit
 }
 nodes [
  Add(
   A=TextureCoordinate()
   B=Multiply(
    A=NamedRerouteUsage()
    B=ScalarParameter() {defaultValue=0.2, group=UV Distortion, parameterName=Flame Distortion Strength, sliderMax=1, sliderMin=-1}
   ) {constB=0.1}
  )
  n1 = Add(
   A=Multiply(A=TextureCoordinate(), B=ScalarParameter() {defaultValue=2, parameterName=Vertical Gradient Tiling, sliderMax=3})
   B=NamedRerouteUsage()
  )
  n2 = CurveAtlasRowParameter(CurveTime=Saturate(None=ComponentMask(None=@n1) {g=true})) {
   atlas=/Game/Athena/LinearColorCurves/VFX_CurveAtlas_Toon
   curve=/Game/Weapons/FORT_Melee/Pickaxe_EarthLane/FX/CC_EarthLane
   defaultValue=10
   parameterName=Curve_Color
  }
  n3 = ComponentMask(None=TextureCoordinate()) {g=true}
  n4 = NamedRerouteUsage()
  n5 = ScalarParameter() {defaultValue=0.1, group=UV Distortion, parameterName=Mask Distortion Strength}
  n6 = MaterialFunctionCall!(UV_Flame, UV_Mask=Add(A=TextureCoordinate(), B=Multiply(A=@n4, B=@n5) {constB=0})) {materialFunction=&/MF_Flame}
  n7 = SmoothStep(Value=Multiply(A=SmoothStep(Value=@n3), B=@n6.Result)) {constMax=1.5, constMin=0.1}
  n8 = CurveAtlasRowParameter(CurveTime=@n7) {
   atlas=&VFX_CurveAtlas_Toon
   curve=&CC_EarthLane
   defaultValue=10
   parameterName=Curve_Color
  }
  n9 = Multiply(A=@n7, B=ScalarParameter() {defaultValue=1, parameterName=Emissive, sliderMax=2})
  n10 = Add(
   A=Multiply(A=Multiply(A=@n2, B=@n2), B=ScalarParameter() {defaultValue=0.05, parameterName=Vertical Gradient Bloom, sliderMax=1})
   B=Multiply(A=@n8, B=Add(A=@n9) {constB=0.5})
  )
  n11 = VectorParameter() {parameterName=TOD color, defaultValue=<rgba 1 1 1 0.5>}
  n12 = MaterialFunctionCall(Value_Morning=@n11.R, Value_Day=@n11.R, Value_Evening=@n11.R, Value_Night=@n11.R) {materialFunction=&Scalar_Time_of_Day}
  n13 = ParticleColor()
  n14 = ScalarParameter() {parameterName=useParticleColor}
  MaterialFunctionCall(Emissive=Multiply(A=Multiply(A=@n10, B=@n12.Result), B=LinearInterpolate(B=@n13.RGB, Alpha=@n14) {constA=1})) {materialFunction=&MF_FX_EmissiveAdjustments}
  n15 = SmoothStep(Value=@n6.Result) {constMax=0.3}
  n16 = DepthFade(Opacity=@n15, FadeDistance=ScalarParameter() {defaultValue=35, parameterName=Depth Fade Distance, sliderMax=100}) {fadeDistanceDefault=25}
  n17 = ShadingPathSwitch(Default=@n16, Deferred=@n16, Forward=@n15, Mobile=@n15)
  n18 = ScalarParameter() {defaultValue=4, parameterName=Opacity, sliderMax=1}
  n19 = ScalarParameter() {defaultValue=4, parameterName="FadeSensitivity (optional) (S)"}
  Multiply(
   A=Multiply(
    A=Multiply(A=@n17, B=@n18)
    B=MaterialFunctionCall("FadeSensitivity (optional)"=@n19) {materialFunction=&FadeBasedOnViewAngle}.Result
   )
   B=LinearInterpolate(B=ParticleColor().A, Alpha=ScalarParameter() {parameterName=useParticleColor}) {constA=1}
  )
  n20 = ScalarParameter() {group=UV Distortion, parameterName=UV Speed X, sliderMax=5, sliderMin=-5}
  n21 = ScalarParameter() {defaultValue=2, group=UV Distortion, parameterName=UV Speed Y, sliderMax=5, sliderMin=-5}
  n22 = ScalarParameter() {defaultValue=1, group=UV Distortion, parameterName=UV Scale X, sliderMax=5}
  n23 = ScalarParameter() {defaultValue=0.5, group=UV Distortion, parameterName=UV Scale Y, sliderMax=5}
  n24 = ScalarParameter() {group=UV Distortion, parameterName=UV Speed X 2, sliderMax=5, sliderMin=-5}
  n25 = ScalarParameter() {defaultValue=1, group=UV Distortion, parameterName=UV Speed Y 2, sliderMax=5, sliderMin=-5}
  n26 = ScalarParameter() {defaultValue=1, group=UV Distortion, parameterName=UV Scale X 2, sliderMax=5}
  n27 = ScalarParameter() {defaultValue=0.25, group=UV Distortion, parameterName=UV Scale Y 2, sliderMax=5}
  NamedRerouteDeclaration(None=MaterialFunctionCall!(
    UV Speed=AppendVector(A=@n20, B=@n21)
    UV Scale=AppendVector(A=@n22, B=@n23)
    UV2 Speed=AppendVector(A=@n24, B=@n25)
    UV2 Scale=AppendVector(A=@n26, B=@n27)
    U Offset=DynamicParameter() {defaultValue=<rgba 0 1 0 1>, paramNames [UV1 U Offset, Dive, Camera Offset, Param4]}.UV1 U Offset
   ) {materialFunction=&/MF_FlameDistortion}.Result) {name=Distortion, nodeColor=<rgba 1 0.3281 0 1>}
  StaticSwitchParameter(
   True=MaterialFunctionCall!(UVs) {materialFunction=&/MF_InterpolateWPO}.Result
   False=Constant()
  ) {defaultValue=true, parameterName=UseInterpolationWPO}
 ]
}
